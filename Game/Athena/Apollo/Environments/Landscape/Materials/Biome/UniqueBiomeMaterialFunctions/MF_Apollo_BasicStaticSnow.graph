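# Material function for Apollo landscape biome materials that lays static snow over base color, behind a static switch.
# An upward mask, sharpened from the Z of input normal blended toward world vertex normal, is scaled by custom primitive data s19Snow.
# That mask lerps base color to a snow color, with a separate mobile/forward color.
# A hardcoded debug toggle can tint zero-snow primitives; the NormalLerp, Pow3 and Mod3 parameters are unconnected.
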MaterialFunction /Game/Athena/Apollo/Environments/Landscape/Materials/Biome/UniqueBiomeMaterialFunctions/MF_Apollo_BasicStaticSnow
graph {nodes [
  n1 = NamedRerouteUsage()
  n2 = VectorParameter() {group=BasicStaticSnow, parameterName=StaticSnow-BaseColorMobile, defaultValue=<rgba 0.35 0.35 0.35 1>}
  n3 = ScalarParameter() {bUseCustomPrimitiveData=true, parameterName=s19Snow, primitiveDataIndex=6}
  n4 = ComponentMask(None=LinearInterpolate(A=NamedRerouteUsage(), B=VertexNormalWS(), Alpha=Constant() {r=0.7})) {b=true}
  n5 = Multiply(A=Power(Base=Saturate(None=@n4), Exp=Constant() {r=5}), B=Constant() {r=1.2})
  n6 = LinearInterpolate(
   A=NamedRerouteUsage()
   B=ShadingPathSwitch(
    Default=VectorParameter() {group=BasicStaticSnow, parameterName=StaticSnow-BaseColor, defaultValue=<rgba 0.56 0.6355 0.7 1>}.RGB
    Forward=@n2.RGB
    Mobile=@n2.RGB
   )
   Alpha=Multiply(A=@n3, B=Saturate(None=@n5))
  )
  n7 = Constant()
  n8 = Constant()
  n9 = Constant() {r=1}
  n10 = Constant() {r=1}
  n11 = StaticSwitch(
   True=LinearInterpolate(
    A=@n6
    B=VectorParameter() {parameterName=StaticSnow-DebugColor, defaultValue=<rgba 1 0.3103 0 1>}.RGB
    Alpha=If(A=@n3, B=@n7, "A > B"=@n8, "A == B"=@n9, "A < B"=@n10)
   )
   False=@n6
   Value=StaticBool() {desc="Set me to true to be easier to debug!"}
  )
  FunctionOutput(None=StaticSwitchParameter(True=SetMaterialAttributes!(MaterialAttributes=@n1, BaseColor=@n11) {attributeSetTypes [#1]}, False=@n1) {defaultValue=true, group=BasicStaticSnow, parameterName=EnableBasicStaticSnow})
  n12 = GetMaterialAttributes(None=FunctionInput() {inputType=FunctionInput_MaterialAttributes}) {attributeGetTypes [#1, #2]}
  NamedRerouteDeclaration(None=@n12.BaseColor) {name=InputDiffuse, nodeColor=<rgba 0.5 0 1 1>}
  NamedRerouteDeclaration(None=@n12.MaterialAttributes) {name=InputMaterial, nodeColor=<rgba 0.5 0 1 1>}
  NamedRerouteDeclaration(None=@n12.Normal) {name=InputNormal, nodeColor=<rgba 0.5 0 1 1>}
  ScalarParameter() {defaultValue=0.7, group=BasicStaticSnow, parameterName=StaticSnow-NormalLerp}
  ScalarParameter() {defaultValue=5, group=BasicStaticSnow, parameterName=StaticSnow-Pow3}
  ScalarParameter() {defaultValue=1.2, group=BasicStaticSnow, parameterName=StaticSnow-Mod3}
 ]}
