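# Material function for Apollo natural rocks that adds optional puddles and blends rock attributes into supplied ground color, roughness, specular and normal.
# Puddles darken, smooth and flatten the surface through a height lerp of blue vertex color, world-projected noise and CustomData0.
# Ground blending follows the World HM input through another height lerp, with GroundMesh and puddle switches altering the mask.
# The RandomColors tint branch is not connected to the output.
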MaterialFunction /Game/Environments/Apollo/Rocks/Natural_Rocks/Materials/MF_RocksColors
graph {nodes [
  n1 = GetMaterialAttributes(None=FunctionInput() {inputType=FunctionInput_MaterialAttributes, sortPriority=-1}) {attributeGetTypes [
    #1
    #2
    #3
    #4
   ]}
  n2 = SetMaterialAttributes!(
   MaterialAttributes=@n1.MaterialAttributes
   BaseColor=Multiply(A=ScalarParameter() {defaultValue=0.5, group=Puddles, parameterName=Puddles Value, sortPriority=1}, B=@n1.BaseColor) {constA=0.35}
   Roughness=Constant() {r=0.025}
   Specular=Multiply(B=@n1.Specular) {constA=0.2}
   Normal=LinearInterpolate(A=@n1.Normal, B=Constant3Vector() {constant=<rgba 0 0 1 0>}) {constAlpha=0.95}
  ) {attributeSetTypes [
    #1
    #5
    #3
    #2
   ]}
  n3 = WorldPosition()
  n4 = TextureSample(UVs=Divide(
    A=ComponentMask(None=@n3.XYZ) {g=true, r=true}
    B=ScalarParameter() {defaultValue=320, group=Puddles, parameterName=Puddles Scale, sortPriority=2}
   )) {samplerSource=SSM_Wrap_WorldGroupSettings, samplerType=SAMPLERTYPE_Grayscale, texture=&LowResBlurredNoise}
  n5 = MaterialFunctionCall!(Diffuse 1, Diffuse 2, Transistion Phase=VertexColor().B, Height Texture 1=@n4.B, Height Texture 2=@n1.CustomData0, Contrast=Constant() {r=3}) {materialFunction=/Engine/Functions/Engine_MaterialFunctions02/Texturing/HeightLerpWithTwoHeightMaps}
  n6 = GetMaterialAttributes(None=StaticSwitchParameter(True=BlendMaterialAttributes(A=@n1.MaterialAttributes, B=@n2, Alpha=@n5.Lerp Alpha No Contrast), False=@n1.MaterialAttributes) {desc=Blue Channel Vertex Color, group=Puddles, parameterName=Puddles}) {attributeGetTypes [
    #1
    #5
    #3
    #2
    #4
   ]}
  n7 = FunctionInput() {bUsePreviewValueAsDefault=true, inputName=Ground Color, sortPriority=2, previewValue=<1 0 0 1>}
  n8 = ComponentMask(None=VertexNormalWS()) {b=true}
  n9 = FunctionInput() {bUsePreviewValueAsDefault=true, inputName=World HM, inputType=FunctionInput_Scalar, sortPriority=6, previewValue=<1 0 0 1>}
  n10 = Add(
   A=Divide(A=@n9, B=ScalarParameter() {defaultValue=155, group=RVT, parameterName=DistanceBlend}) {constB=150}
   B=ScalarParameter() {defaultValue=0.1571, group=RVT, parameterName=DistanceBlendOffset}
  )
  n11 = StaticSwitchParameter(True=Min(A=VertexColor().G, B=Max(A=@n8, B=@n10)), False=@n10) {group=RVT, parameterName=GroundMesh}
  n12 = FunctionInput() {bUsePreviewValueAsDefault=true, inputName=HM Input, inputType=FunctionInput_Scalar, sortPriority=5, previewValue=<0.5 0 0 1>}
  n13 = MaterialFunctionCall(
   Transition Phase=Saturate(None=OneMinus(None=@n11))
   Height Texture=OneMinus(None=StaticSwitchParameter(True=@n6.CustomData0, False=@n12) {group=RVT, parameterName=GroundMesh})
  ) {materialFunction=&HeightLerp}
  n14 = Saturate(None=StaticSwitchParameter(True=Subtract(A=@n13.Lerp Alpha No Contrast, B=@n5.Lerp Alpha No Contrast), False=@n13.Lerp Alpha No Contrast) {group=Puddles, parameterName=Puddles})
  n15 = LinearInterpolate(A=@n6.BaseColor, B=@n7, Alpha=@n14)
  n16 = FunctionInput() {bUsePreviewValueAsDefault=true, inputName=Ground Roughness, inputType=FunctionInput_Scalar, sortPriority=3, previewValue=<3 0 0 1>}
  n17 = LinearInterpolate(A=@n6.Roughness, B=@n16, Alpha=@n14)
  n18 = FunctionInput() {bUsePreviewValueAsDefault=true, inputName=Ground Specular, inputType=FunctionInput_Scalar, sortPriority=3, previewValue=<3 0 0 1>}
  n19 = LinearInterpolate(A=@n6.Specular, B=@n18, Alpha=@n14)
  n20 = FunctionInput() {bUsePreviewValueAsDefault=true, inputName=Ground Normal, sortPriority=4, previewValue=<0 0 1 1>}
  n21 = ScalarParameter() {defaultValue=12, parameterName=DistanceN}
  n22 = LinearInterpolate(A=@n20, B=@n6.Normal, Alpha=Saturate(None=Divide(A=@n9, B=@n21)))
  n23 = SetMaterialAttributes(MaterialAttributes=@n6.MaterialAttributes, BaseColor=@n15, Roughness=@n17, Specular=@n19, Normal=@n22) {attributeSetTypes [
    #1
    #5
    #3
    #2
   ]}
  n24 = SetMaterialAttributes(MaterialAttributes=@n6.MaterialAttributes, BaseColor=@n15, Roughness=@n17, Specular=@n19, Normal=@n22) {attributeSetTypes [
    #1
    #5
    #3
    #2
   ]}
  FunctionOutput(None=StaticSwitchParameter(True=@n23, False=@n24) {group=RVT, parameterName=GroundMesh})
  n25 = VectorParameter() {group=Colors, parameterName=RandomColor_A, defaultValue=<rgba 0.5554 0.5898 0.6094 1>}
  n26 = VectorParameter() {group=Colors, parameterName=RandomColor_B, defaultValue=<rgba 0.4792 0.4429 0.3544 1>}
  n27 = ScalarParameter() {group=Colors, parameterName=RockRandomLerpValue}
  StaticSwitchParameter(True=Multiply(A=@n6.BaseColor, B=LinearInterpolate(A=@n25.RGB, B=@n26.RGB, Alpha=@n27)), False=@n6.BaseColor) {defaultValue=true, group=Colors, parameterName=RandomColors}
 ]}
