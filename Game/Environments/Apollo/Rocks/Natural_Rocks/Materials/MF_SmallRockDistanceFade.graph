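# Material function for small Apollo rocks that shades them procedurally and fades them into the landscape, behind a static switch.
# World-aligned detail textures drive roughness, specular and normals, with base color from per-instance random tints and overlay blending.
# Instance-space height masks blend color toward the landscape runtime virtual texture and normals toward instance up, with quality-switched normal detail.
# Per-instance fade amount sinks rocks through world position offset; Forward and Mobile override base color.
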MaterialFunction /Game/Environments/Apollo/Rocks/Natural_Rocks/Materials/MF_SmallRockDistanceFade
graph {nodes [
  Constant3Vector() {constant=<rgba 0.125 0.1088 0.0959 1>}
  Constant3Vector() {constant=<rgba 0.1 0.0889 0.08 1>}
  Constant3Vector() {constant=<rgba 0.17 0.1473 0.1292 1>}
  Constant() {r=450}
  Constant() {r=0.6}
  Constant() {r=-0.5}
  Constant()
  Constant() {r=1.5}
  Constant()
  Constant() {r=2.05}
  Constant() {r=-1.1}
  Constant() {r=7}
  Constant() {r=-0.8}
  Constant() {r=6500}
  FunctionInput(Preview=Constant3Vector() {constant=<rgba 0.0765 0.1198 0.0236 1>}) {bUsePreviewValueAsDefault=true, inputName=VirtualTextureBaseColorIn, sortPriority=3}
  n1 = Constant3Vector() {constant=<rgba 0 0 1 0>}
  n2 = LinearInterpolate(
   A=NamedRerouteUsage()
   B=VertexNormalWS()
   Alpha=ScalarParameter() {defaultValue=0.5, group=SmallRocksDistanceFade, parameterName=SMDF-VertexNormalBlend}
  )
  n3 = Add(A=@n2, B=NamedRerouteUsage())
  n4 = ComponentMask(None=TransformPosition(None=WorldPosition().XYZ) {transformSourceType=TRANSFORMPOSSOURCE_World, transformType=TRANSFORMPOSSOURCE_Instance}) {b=true}
  n5 = ScalarParameter() {group=SmallRocksDistanceFade, parameterName=SMDF-InsSpaceAdd}
  n6 = Saturate(None=Divide(A=Add(A=@n4, B=@n5) {constB=10}, B=ScalarParameter() {defaultValue=10, group=SmallRocksDistanceFade, parameterName=SMDF-InsSpaceDiv}))
  n7 = Saturate(None=LinearInterpolate(
    A=ScalarParameter() {defaultValue=-0.5, group=SmallRocksDistanceFade, parameterName=SMDF-BlendLerp-MIN}
    B=ScalarParameter() {defaultValue=0.6, group=SmallRocksDistanceFade, parameterName=SMDF-BlendLerp-MAX}
    Alpha=NamedRerouteUsage()
   ))
  n8 = Constant() {r=-20}
  n9 = Constant()
  n10 = PerInstanceFadeAmount()
  n11 = SetMaterialAttributes!(
   MaterialAttributes
   BaseColor=NamedRerouteUsage()
   Normal=LinearInterpolate(
    A=Transform(None=@n1) {desc=This looks goooooood, transformSourceType=TRANSFORMSOURCE_Instance}
    B=QualitySwitch(Default=@n2, Low=@n2, High=@n3, Medium=@n2, Epic=@n3)
    Alpha=Multiply(A=@n6, B=@n7)
   )
   Specular=NamedRerouteUsage()
   WorldPositionOffset=Multiply(A=LinearInterpolate(A=@n8, B=@n9, Alpha=@n10), B=Constant3Vector() {constant=<rgba 0 0 1 0>})
   Roughness=NamedRerouteUsage()
  ) {attributeSetTypes [
    #1
    #2
    #3
    #4
    #5
   ]}
  n12 = SetMaterialAttributes!(MaterialAttributes=@n11, BaseColor=NamedRerouteUsage()) {attributeSetTypes [#1]}
  n13 = FunctionInput() {inputName=Material, inputType=FunctionInput_MaterialAttributes}
  FunctionOutput(None=StaticSwitchParameter(True=ShadingPathSwitch(Default=@n11, Forward=@n12, Mobile=@n12), False=@n13) {group=SmallRocksDistanceFade, parameterName=UseSmallRockDistanceFade})
  NamedRerouteDeclaration(None=SphereMask(A=CameraPositionWS(), B=WorldPosition().XYZ, Radius=Constant() {r=10000}, Hardness=Constant())) {name=CameraDistanceFade, nodeColor=<rgba 1 0 1 1>}
  n14 = MaterialFunctionCall!(
   Diffuse Texture Object=TextureObject() {texture=/Game/Environments/Apollo/Rocks/Natural_Rocks/Textures/T_RockDetails_02Db}
   Spec Texture Object
   TextureSize=ScalarParameter() {defaultValue=600, group=SmallRocksDistanceFade, parameterName=SMDF-DetailTile}
   --------------------------
   Use seperate settings for Z projection
   Use seperate texture for Z projection
   Texture Size Z projection
   Z Plane Diffuse Texture Object
   Z Plane Spec Texture Object
   "--------------------------      "
   ProjectionTransitionContrast=ScalarParameter() {defaultValue=1, group=SmallRocksDistanceFade, parameterName=SMDF-PTC}
   Use Texture Alpha Heightmap to Refine the Transistion
   HeightLerpTransitionContrast
   PickTheDiffuseHeightChannel
   TheDiffuseHeightChannel
   "--------------------------       "
   NormalTextureObject=TextureObject() {samplerType=SAMPLERTYPE_Normal, texture=/Game/Environments/Apollo/Rocks/Natural_Rocks/Textures/T_RockDetails_02_N}
   World Space Normals=StaticBool()
   "--------------------------  "
   Export Float 4 Diffuse
   Export Float 4 Spec
   "--------------------------     "
   WorldPosition
   World Space Normal
  ) {materialFunction=/Engine/Functions/Engine_MaterialFunctions02/Texturing/WorldAlignedTextures_Complex}
  n15 = VectorParameter() {group=SmallRocksDistanceFade, parameterName=SMDF-NrmModVec3, defaultValue=<rgba 1 1 0 1>}
  NamedRerouteDeclaration(None=Transform(None=Multiply(A=@n14.Normal XYZ Texture, B=@n15.RGB))) {name=Detail-Normal, nodeColor=<rgba 1 0 1 1>}
  NamedRerouteDeclaration(None=ComponentMask(None=@n14.XYZ Diffuse Texture) {r=true}) {name=Detail-Mask, nodeColor=<rgba 1 0 1 1>}
  NamedRerouteDeclaration(None=ComponentMask(None=@n14.XYZ Diffuse Texture) {b=true, g=true, r=true}) {name=Detail-Diffuse, nodeColor=<rgba 1 0 1 1>}
  n16 = GetMaterialAttributes(None=@n13) {attributeGetTypes [
    #1
    #2
    #3
    #5
   ]}
  NamedRerouteDeclaration(None=@n16.Normal) {name=NormalIn, nodeColor=<rgba 1 0 1 1>}
  NamedRerouteDeclaration(None=@n16.BaseColor) {name=DiffuseIn, nodeColor=<rgba 1 0 1 1>}
  n17 = LinearInterpolate(
   A=ScalarParameter() {defaultValue=-0.25, group=SmallRocksDistanceFade, parameterName=SMDF-Detail R Min, sortPriority=5}
   B=ScalarParameter() {defaultValue=1.5, group=SmallRocksDistanceFade, parameterName=SMDF-Detail R Max, sortPriority=6}
   Alpha=NamedRerouteUsage()
  )
  NamedRerouteDeclaration(None=LinearInterpolate(
    A=ScalarParameter() {defaultValue=1, group=SmallRocksDistanceFade, parameterName=SMDF-FarRoughness, sortPriority=5}
    B=Saturate(None=@n17)
    Alpha=NamedRerouteUsage()
   )) {name=Rough, nodeColor=<rgba 1 0 1 1>}
  n18 = ScalarParameter() {defaultValue=2, group=SmallRocksDistanceFade, parameterName=SMDF-SpecDet1, sortPriority=5}
  n19 = NamedRerouteUsage()
  n20 = Saturate(None=Add(
    A=ScalarParameter() {defaultValue=-0.5, group=SmallRocksDistanceFade, parameterName=SMDF-SpecDet2, sortPriority=5}
    B=Multiply(A=@n18, B=@n19)
   ))
  n21 = Saturate(None=LinearInterpolate(
    A=ScalarParameter() {defaultValue=1.5, group=SmallRocksDistanceFade, parameterName=SMDF-SpecDet3, sortPriority=5}
    B=ScalarParameter() {defaultValue=-1, group=SmallRocksDistanceFade, parameterName=SMDF-SpecDet4, sortPriority=5}
    Alpha=@n19
   ))
  NamedRerouteDeclaration(None=Multiply(A=Multiply(A=@n20, B=@n21), B=ScalarParameter() {defaultValue=1, group=SmallRocksDistanceFade, parameterName=SMDF-SpecMod, sortPriority=5})) {name=Spec, nodeColor=<rgba 1 0 1 1>}
  n22 = RuntimeVirtualTextureSampleParameter() {
   desc="Both RVT textures need to be set correctly!"
   group=RVT
   materialType=BaseColor_Normal_Specular_YCoCg
   parameterName=RVT-TextureSample
   virtualTexture=/Game/Athena/Apollo/Maps/Landscape/VT/RVT_Apollo_Landscape
  }
  n23 = Multiply(A=NamedRerouteUsage(), B=@n6)
  n24 = MaterialFunctionCall(
   Base=MaterialFunctionCall(
    1=VectorParameter() {group=SmallRocksDistanceFade, parameterName=SMDF-Color1, defaultValue=<rgba 0.15 0.1305 0.1151 1>}.RGB
    2=VectorParameter() {group=SmallRocksDistanceFade, parameterName=SMDF-Color2, defaultValue=<rgba 0.14 0.1244 0.112 1>}.RGB
    3=VectorParameter() {group=SmallRocksDistanceFade, parameterName=SMDF-Color3, defaultValue=<rgba 0.21 0.1819 0.1596 1>}.RGB
    A=PerInstanceRandom()
   ) {materialFunction=&Lerp_Multiple_Float3}.Lerp 3 Inputs
   Blend=NamedRerouteUsage()
  ) {materialFunction=&Blend_Overlay}
  NamedRerouteDeclaration(None=MaterialFunctionCall!(NoRVT=@n24.Result, RVT=LinearInterpolate(A=@n22.BaseColor, B=@n24.Result, Alpha=@n23)) {materialFunction=/Game/Packages/Fortress_SharedMaterials/Base_Material_Functions/FN_RVT_Switch_Color}.Result) {name=Diffuse, nodeColor=<rgba 1 0 1 1>}
  ScalarParameter() {group=SmallRocksDistanceFade, parameterName=SMDF-RVT-Mip-test}
  StaticBool() {value=true}
  TextureObject() {texture=/Game/Environments/Apollo/Rocks/Natural_Rocks/Textures/T_RockDetails_02D}
 ]}
