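# Material function that layers world-aligned detail normals for sides and top onto incoming material attributes, apparently for Apollo natural rocks.
# Two WorldAlignedTextures_Complex calls supply XY side and Z top projections; vertex normal Z fades side strength upward and blends bottom-to-top strength.
# HeightLerp mixes them using absolute normal Z and the top diffuse alpha as height.
# The result's XY is added to the existing normal, then renormalized.
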MaterialFunction /Game/Environments/Apollo/Rocks/Natural_Rocks/Materials/MF_WSNormals
graph {nodes [
  n1 = GetMaterialAttributes(None=FunctionInput() {inputType=FunctionInput_MaterialAttributes}) {attributeGetTypes [#1]}
  n2 = ScalarParameter() {defaultValue=1024, group=WS Normals, parameterName=WS_Side_Scale, sortPriority=7}
  n3 = TextureObjectParameter() {
   group=WS Normals
   parameterName=WS_N_Sides
   samplerType=SAMPLERTYPE_Normal
   sortPriority=6
   texture=/Game/Global/Textures/Default/Tileable_Patterns/T_Snow01_N
  }
  n4 = StaticBool()
  n5 = ComponentMask(None=VertexNormalWS()) {b=true}
  n6 = Saturate(None=@n5)
  n7 = LinearInterpolate(
   A=Multiply(
    A=MaterialFunctionCall!(
     Diffuse Texture Object
     Spec Texture Object
     TextureSize=@n2
     --------------------------
     Use seperate settings for Z projection
     Use seperate texture for Z projection
     Texture Size Z projection
     Z Plane Diffuse Texture Object
     Z Plane Spec Texture Object
     "--------------------------      "
     ProjectionTransitionContrast
     Use Texture Alpha Heightmap to Refine the Transistion
     HeightLerpTransitionContrast
     PickTheDiffuseHeightChannel
     TheDiffuseHeightChannel
     "--------------------------       "
     NormalTextureObject=@n3
     World Space Normals=@n4
     "--------------------------  "
     Export Float 4 Diffuse
     Export Float 4 Spec
     "--------------------------     "
     WorldPosition
     World Space Normal
    ) {materialFunction=/Engine/Functions/Engine_MaterialFunctions02/Texturing/WorldAlignedTextures_Complex}.Normal XY Texture
    B=ScalarParameter() {defaultValue=8, group=WS Normals, parameterName=Sides N Value}
   )
   Alpha=@n6
  ) {constB=0}
  n8 = TextureObjectParameter() {
   group=WS Normals
   parameterName=WS_N_Top
   samplerType=SAMPLERTYPE_Normal
   sortPriority=1
   texture=/Game/Environments/Apollo/Sets/Terrain/Textures/T_Apollo_Cliff_N_02
  }
  n9 = MaterialFunctionCall(
   Diffuse Texture Object=TextureObjectParameter() {group=WS Normals, parameterName=WS_D_Top_Height, texture=/Game/Environments/Apollo/Sets/Terrain/Textures/T_Apollo_Cliff_D_02}
   TextureSize=ScalarParameter() {defaultValue=1785.5282, group=WS Normals, parameterName=WS_Top_Scale, sortPriority=2}
   NormalTextureObject=@n8
   World Space Normals=StaticBool()
   Export Float 4 Diffuse=StaticBool() {value=true}
  ) {materialFunction=&WorldAlignedTextures_Complex}
  n10 = Multiply(
   A=LinearInterpolate(
    A=ScalarParameter() {defaultValue=1, group=WS Normals, parameterName=Bottom N Value, sortPriority=4}
    B=ScalarParameter() {group=WS Normals, parameterName=Top N Value, sortPriority=3}
    Alpha=@n6
   ) {constB=0}
   B=@n9.Normal Z Texture
  )
  n11 = MaterialFunctionCall(
   Transition Phase=Abs(None=@n5)
   Height Texture=ComponentMask(None=@n9.Z Diffuse Texture) {a=true}
   Contrast=ScalarParameter() {defaultValue=1, group=WS Normals, parameterName=Top/Sides Contrast, sortPriority=5}
  ) {materialFunction=&HeightLerp}
  n12 = Add(A=@n1.Normal, B=Multiply(A=LinearInterpolate(A=@n7, B=@n10, Alpha=@n11.Alpha), B=Constant3Vector() {constant=<rgba 1 1 0 1>}))
  FunctionOutput(None=SetMaterialAttributes!(MaterialAttributes=@n1.MaterialAttributes, Normal=Normalize(VectorInput=@n12)) {attributeSetTypes [#1]})
 ]}
