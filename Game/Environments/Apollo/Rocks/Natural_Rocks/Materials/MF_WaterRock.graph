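# Material function that gives rock materials an optional wet look through the WaterRock static switch in the WetRock group.
# It darkens base color and lowers roughness and specular on the incoming material attributes.
# A saturated, scaled world-space vertex-normal Z mask blends from this wet result toward the blend's B input on upward faces, though B is left unconnected.
# With the switch off, the input passes through unchanged.
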MaterialFunction /Game/Environments/Apollo/Rocks/Natural_Rocks/Materials/MF_WaterRock
graph {nodes [
  n1 = FunctionInput() {inputType=FunctionInput_MaterialAttributes}
  n2 = GetMaterialAttributes(None=@n1) {attributeGetTypes [#1, #2, #3]}
  n3 = SetMaterialAttributes!(
   MaterialAttributes=@n2.MaterialAttributes
   BaseColor=Multiply(A=@n2.BaseColor) {constB=0.5}
   Roughness=Multiply(A=@n2.Roughness) {constB=0.25}
   Specular=Multiply(A=@n2.Specular) {constB=0.3}
  ) {attributeSetTypes [#1, #2, #3]}
  n4 = VertexNormalWS()
  n5 = Saturate(None=Multiply(A=ComponentMask(None=@n4) {b=true}, B=Constant() {r=0.8}))
  FunctionOutput(None=StaticSwitchParameter(True=BlendMaterialAttributes(A=@n3, Alpha=@n5), False=@n1) {group=WetRock, parameterName=WaterRock})
 ]}
