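# Nanite-ready master material for Figment environment rocks, based on Apollo natural rock textures with UV or world-aligned projection.
# Chained material attributes add switchable gradient, tint, emissive flicker, grass top, detail, moss and ocean wet line.
# Runtime virtual texture color and height samples blend rocks into terrain, skipped on low quality and HLOD proxies.
# Final functions apply optional animation, damage, distance fade, snow, landscape fire (Curie) and creative move-tool effects.
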Material /Figment_S06/Athena/Environments/Materials/M_ApolloRocks_Figment
graph {
 properties {
  bTangentSpaceNormal=false
  bUseMaterialAttributes=true
  bUsedWithInstancedStaticMeshes=true
  bUsedWithNanite=true
  ditheredLODTransition=true
  pixelDepthOffsetMode=PDOM_Legacy
  usedShadingModels=Default Lit
 }
 nodes [
  Constant() {r=8}
  LinearInterpolate(
   A=ScalarParameter() {defaultValue=9, parameterName=RVSides}
   B=ScalarParameter() {defaultValue=8, group=RVT, parameterName=RVTMipBias}
   Alpha=ComponentMask(None=VertexNormalWS()) {b=true}
  )
  n1 = RuntimeVirtualTextureSampleParameter() {
   desc="Both RVT textures need to be set correctly!"
   group=RVT
   materialType=BaseColor_Normal_Specular_Mask_YCoCg
   parameterName=RVT-TextureSample
   virtualTexture=/FigmentCore/Athena/Prototype/Terrain/RVT_Figment
  }
  n2 = RuntimeVirtualTextureSampleParameter() {
   desc="Both RVT textures need to be set correctly!"
   group=RVT
   materialType=WorldHeight
   parameterName=RVT-WorldHeightSample
   virtualTexture=/FigmentCore/Athena/Prototype/Terrain/RVT_Figment_Height
  }
  n3 = Subtract(A=ComponentMask(None=WorldPosition() {worldPositionShaderOffset=WPT_ExcludeAllShaderOffsets}.XYZ) {b=true}, B=@n2.WorldHeight)
  n4 = VectorParameter() {group=WS Gradient, parameterName=WS GradientBottom, sortPriority=4, defaultValue=<rgba 0.3594 0.304 0.1834 0>}
  n5 = ActorPositionWS()
  n6 = ScalarParameter() {defaultValue=-300, group=WS Gradient, parameterName=Gradient Start, sortPriority=2}
  n7 = ComponentMask(None=Subtract(A=WorldPosition().XYZ, B=StaticSwitchParameter(True=@n5, False=@n6) {group=WS Gradient, parameterName=Use Actor Location, sortPriority=1})) {b=true}
  n8 = ScalarParameter() {defaultValue=3000, group=WS Gradient, parameterName=Gradient, sortPriority=3}
  n9 = LinearInterpolate(
   A=@n4.RGB
   B=VectorParameter() {group=WS Gradient, parameterName=WS GradientTop, sortPriority=5, defaultValue=<rgba 1 1 1 1>}.RGB
   Alpha=Saturate(None=Divide(A=@n7, B=@n8))
  )
  n10 = TextureObjectParameter() {group=Rock Textures, parameterName=Diffuse, texture=/Game/Environments/Apollo/Rocks/Natural_Rocks/Textures/T_NaturalRock_AB_D}
  n11 = TextureObjectParameter() {
   group=Rock Textures
   parameterName=SpecularMasks
   samplerType=SAMPLERTYPE_Masks
   texture=/Game/Environments/Apollo/Rocks/Natural_Rocks/Textures/T_NaturalRock_AB_S
  }
  n12 = TextureObjectParameter() {
   group=Rock Textures
   parameterName=Normals
   samplerType=SAMPLERTYPE_Normal
   texture=/Game/Environments/Apollo/Rocks/Natural_Rocks/Textures/T_NaturalRock_AB_N
  }
  n13 = MaterialFunctionCall!(
   Diffuse Texture Object=@n10
   Spec Texture Object=@n11
   TextureSize=ScalarParameter() {defaultValue=512, group=Rock Textures, parameterName=Base Textur Scale WS}
   --------------------------
   Use seperate settings for Z projection
   Use seperate texture for Z projection
   Texture Size Z projection
   Z Plane Diffuse Texture Object
   Z Plane Spec Texture Object
   "--------------------------      "
   ProjectionTransitionContrast
   Use Texture Alpha Heightmap to Refine the Transistion
   HeightLerpTransitionContrast
   PickTheDiffuseHeightChannel
   TheDiffuseHeightChannel
   "--------------------------       "
   NormalTextureObject=@n12
   World Space Normals=StaticBool()
   "--------------------------  "
   Export Float 4 Diffuse=StaticBool() {value=true}
   Export Float 4 Spec
   "--------------------------     "
   WorldPosition
   World Space Normal
  ) {materialFunction=/Engine/Functions/Engine_MaterialFunctions02/Texturing/WorldAlignedTextures_Complex}
  n14 = StaticSwitchParameter(True=@n13.Z Diffuse Texture, False=TextureSample(Tex=@n10).RGBA) {group=Rock Textures, parameterName="Base Texture WS Z "}
  n15 = ComponentMask(None=@n14) {b=true, g=true, r=true}
  n16 = Multiply(A=VectorParameter() {group=Rock Textures, parameterName=Diffuse Color, defaultValue=<rgba 0.5 0.5 0.5 1>}.RGB, B=@n15)
  n17 = MaterialFunctionCall!(DarknessAmount) {materialFunction=/Game/Athena/Apollo/Environments/Landscape/Materials/Biome/CommonMaterials/MF_HeightBasedDarkenCliff}
  n18 = NamedRerouteDeclaration(None=ComponentMask(None=@n14) {a=true}) {name=Alpha, nodeColor=<rgba 0.7812 1 0 1>}
  n19 = ScalarParameter() {group=Rock Textures, parameterName=Diffuse Mask Bias}
  n20 = StaticSwitchParameter(
   True=@n16
   False=LinearInterpolate(
    A=@n15
    B=Multiply(A=@n17.Result, B=@n16)
    Alpha=MaterialFunctionCall(In=@n18, Contrast=@n19) {materialFunction=&CheapContrast}.Result
   )
  ) {group=Rock Textures, parameterName=DiffuseClean}
  n21 = MaterialFunctionCall(Float3=StaticSwitchParameter(True=@n13.Z Spec, False=TextureSample(Tex=@n11).RGB) {group=Rock Textures, parameterName="Base Texture WS Z "}) {materialFunction=&BreakOutFloat3Components}
  n22 = VectorParameter() {desc=Enable SSS in instance, group=SSS, parameterName=SSS Color, defaultValue=<rgba 0.6302 0.2429 0.2429 0>}
  n23 = ScalarParameter() {defaultValue=0.5, desc=Enable SSS in instance, group=SSS, parameterName=SSS Color Vs Diffuse}
  n24 = SetMaterialAttributes!(
   MaterialAttributes
   Base Color=StaticSwitchParameter(True=Multiply(A=@n9, B=@n20), False=@n20) {group=WS Gradient, parameterName=WS Gradient}
   Specular=@n21.R
   Roughness=@n21.R
   Normal=NamedRerouteUsage()
   Subsurface Color=LinearInterpolate(A=@n15, B=@n22.RGB, Alpha=Multiply(A=@n18, B=@n23))
   Opacity=Constant() {r=0.5}
   Custom Data 0=NamedRerouteUsage()
   Metallic=@n21.R
  ) {attributeSetTypes [
    #1
    #2
    #3
    #4
    #5
    #6
    #7
    #8
   ]}
  n25 = VectorParameter() {group=Emissive, parameterName=EmissiveColor, defaultValue=<rgba 1 0 0 0>}
  n26 = ScalarParameter() {defaultValue=8, group=Emissive, parameterName=EmissivePower}
  n27 = VectorParameter() {group=Emissive, parameterName=EmissiveColor_1, defaultValue=<rgba 0.5518 0.1107 0.6979 0>}
  n28 = ScalarParameter() {defaultValue=5, group=Emissive, parameterName=EmissivePower_1}
  n29 = TextureSampleParameter2D(UVs=Multiply(A=TextureCoordinate(), B=ScalarParameter() {defaultValue=1, group=Emissive, parameterName=Scale Emissive UV})) {
   group=Emissive
   parameterName=Emissive
   samplerSource=SSM_Clamp_WorldGroupSettings
   sortPriority=0
   texture=&T_BlacksRGB
  }
  n30 = LinearInterpolate(B=LinearInterpolate(A=Multiply(A=@n25.RGB, B=@n26), B=Multiply(A=@n27.RGB, B=@n28), Alpha=@n29.R) {constB=0}, Alpha=@n29.R)
  n31 = StaticSwitchParameter(
   True=MaterialFunctionCall!(MA=SetMaterialAttributes!(MaterialAttributes=@n24, Emissive Color=@n30) {attributeSetTypes [#9]}) {materialFunction=/Game/Environments/Sets/PrincessCastle/Props/Materials/MF_GlowFlicker}.Result
   False=@n24
  ) {group=Emissive, parameterName=EmissiveTrue, sortPriority=-1}
  n32 = GetMaterialAttributes(None=@n31) {attributeGetTypes [#1]}
  n33 = LinearInterpolate(
   A=VectorParameter() {group=Colors, parameterName=RandomColor_A, defaultValue=<rgba 0.5554 0.5898 0.6094 1>}.RGB
   B=VectorParameter() {group=Colors, parameterName=RandomColor_B, defaultValue=<rgba 0.4792 0.4429 0.3544 1>}.RGB
   Alpha=ScalarParameter() {group=Colors, parameterName=RockRandomLerpValue}
  )
  n34 = StaticSwitchParameter(
   True=SetMaterialAttributes!(MaterialAttributes=@n32.MaterialAttributes, Base Color=Multiply(A=@n32.Base Color, B=@n33)) {attributeSetTypes [#1]}
   False=@n31
  ) {defaultValue=true, group=Colors, parameterName=RandomColors}
  n35 = StaticSwitchParameter(True=NamedRerouteUsage(), False=@n34) {desc=Use Base Diffuse Alpha for mask. Up vector is by default, group=Grass, parameterName=GrassTop}
  n36 = MaterialFunctionCall!(In=MaterialFunctionCall!(In=@n35) {materialFunction=/Game/Environments/Apollo/Rocks/Natural_Rocks/Materials/MF_WaterRock}.Result) {materialFunction=/Game/Environments/Apollo/VS/Materials/MF_DetailTexture}
  n37 = GetMaterialAttributes(None=@n36.Result) {attributeGetTypes [#4]}
  n38 = SetMaterialAttributes!(MaterialAttributes=@n37.MaterialAttributes, Normal=Transform(None=@n37.Normal)) {attributeSetTypes [#4]}
  n39 = StaticSwitch(
   True=@n38
   False=MaterialFunctionCall(
    NoRVT=@n38
    RVT=MaterialFunctionCall!(
     In=@n38
     Ground Color=@n1.BaseColor
     Ground Specular=@n1.Specular
     Ground Roughness=@n1.Roughness
     Ground Normal=@n1.Normal
     HM Input=@n36.Texture Only
     World HM=@n3
    ) {materialFunction=/Game/Environments/Apollo/Rocks/Natural_Rocks/Materials/MF_RocksColors}.Result
   ) {materialFunction=&FN_RVT_Switch_Material}.Result
   Value=StaticBoolParameter() {group=WetRock, parameterName=WaterRock}
  )
  n40 = StaticSwitchParameter(
   True=MaterialProxyReplace(Realtime=QualitySwitch(Default=@n39, Low=@n38, High=@n39, Medium=@n39), MaterialProxy=@n38) {desc=Don't use RVT on HLODs}
   False=@n38
  ) {defaultValue=true, group=Rocks RVT, parameterName=UseRVT}
  n41 = StaticSwitchParameter(
   True=MaterialFunctionCall!(WorldSpaceNormals=StaticBool() {value=true}, In=@n40) {materialFunction=/Game/Environments/Apollo/Rocks/Natural_Rocks/Materials/MossSoft2}.Result
   False=@n40
  ) {group=Moss, parameterName=Moss True}
  n42 = GetMaterialAttributes(None=@n41) {attributeGetTypes [#1, #2, #3]}
  n43 = WorldPosition()
  n44 = Subtract(A=@n43.XYZ, B=VectorParameter() {group=HLOD, parameterName=HLODBakeOffset, defaultValue=<rgba 0 0 0 1>}.B)
  n45 = Add(
   A=ComponentMask(None=MaterialProxyReplace(Realtime=@n43.XYZ, MaterialProxy=@n44)) {b=true}
   B=ScalarParameter() {defaultValue=50, desc=Ocean level wet line, group=WetLine, parameterName=WetLineHeight}
  )
  n46 = ScalarParameter() {defaultValue=50, desc=Ocean level wet line, group=WetLine, parameterName=WetLineBlr2}
  n47 = BlendMaterialAttributes(
   A=SetMaterialAttributes!(
    MaterialAttributes=@n42.MaterialAttributes
    Base Color=Multiply(A=@n42.Base Color) {constB=0.4}
    Specular=Multiply(A=@n42.Specular) {constB=0.5}
    Roughness=Multiply(A=@n42.Roughness) {constB=0.3}
   ) {attributeSetTypes [#1, #2, #3]}
   B=@n42.MaterialAttributes
   Alpha=Saturate(None=Divide(A=@n45, B=@n46))
  )
  n48 = StaticSwitchParameter(True=@n41, False=ShadingPathSwitch(Default=@n47, Deferred=@n47, Forward=@n42.MaterialAttributes, Mobile=@n42.MaterialAttributes)) {desc=If not in Apollo Map, group=WetLine, parameterName=DisableOceanLevel}
  n49 = MaterialFunctionCall!(In=@n48) {materialFunction=/Game/Packages/Fortnite_UniqueMaterials_1/MaterialFunctions/Trees/MF_4Level_TreeAnimations}
  n50 = MaterialFunctionCall!(
   Material=MaterialFunctionCall(
    Material_In=StaticSwitchParameter(True=@n49.Result, False=@n48) {parameterName=Animated}
    UseCrackedDamageTextures=StaticBool()
    UseAdditiveWPO=StaticBool() {value=true}
   ) {materialFunction=&MF_FN_DestructibleDamage}.Material_Out
   VirtualTextureBaseColorIn=@n1.BaseColor
  ) {
   desc="Should only be used on landscape\r\ngrass system meshes"
   materialFunction=/Game/Environments/Apollo/Rocks/Natural_Rocks/Materials/MF_SmallRockDistanceFade
  }
  n51 = StaticBool()
  n52 = StaticBool()
  n53 = StaticSwitchParameter(
   True=@n50.Result
   False=MaterialFunctionCall!(
    IsStaticSnow=StaticBoolParameter() {desc="Disconnect this from \"IsStaticSnow\" when using dynamic snow", group=SnowProgression, parameterName=IsStaticSnow}
    UseAlternateStaticSnowMask
    AlternateStaticSnowMask
    "UseRoad&&OcclusionMask"=@n51
    IsForGrassMeshToPushDown
    AdditionalMaskToRemoveSnow
    IsFallingSnowFX
    UseMinimapCoords
    IncreaseSnowUsingSlope=@n52
    UseSmoothBorder
    LerpHighM
    TemporalHeightLerpContrast
    IsLobbyIsland
    UseBR-IslandMask
    IsTree
    IsForRenderTarget
    UseSparkle=@n52
    UseSnowNormals=@n52
    UseAdditionalMaskToRemoveSnow
    IsTangentSpaceInput
    UseVertexNormalsForBasicSnow=StaticBool() {value=true}
    UseTangentSnowNormals=@n51
    Material=@n50.Result
   ) {materialFunction=/Game/Athena/Apollo/Environments/Landscape/Materials/Biome/UniqueBiomeMaterialFunctions/MF_Apollo_Snow_Progression}.Material
  ) {group=WS Normals, parameterName=WS Normals, sortPriority=-1}
  n54 = MaterialFunctionCall(MA In=MaterialFunctionCall!(
    In=MaterialFunctionCall!(In=@n53) {materialFunction=/Game/Athena/Apollo/Environments/Landscape/Materials/Biome/UniqueBiomeMaterialFunctions/MF_Apollo_BasicStaticSnow}.Result
    UseHeightFalloff
    OnFireEmissiveOverride=Constant()
    HeightFalloffRadius
    HeightFalloffHardness
    VariationMasksMobile
    In_Grass
    Is_Stylised_WE
   ) {materialFunction=/Game/Packages/Fortress_SharedMaterials/Base_Material_Functions/MF_Curie_Landscape}.Result) {materialFunction=&MF_CreativeMoveTool}
  MaterialProxyReplace(
   Realtime=@n54.MA Out
   MaterialProxy=SetMaterialAttributes!(MaterialAttributes=@n54.MA Out, Normal=NamedRerouteUsage()) {attributeSetTypes [#4]}
  )
  n55 = MaterialFunctionCall!(CombinedLandscapeLayersMask, BypassUseRVT, BypassBiomeColors) {materialFunction=/Game/Athena/Apollo/Environments/Landscape/Materials/Biome/CommonMaterials/MF_Common_Grass}
  n56 = SetMaterialAttributes!(
   MaterialAttributes
   Base Color=Multiply(A=VectorParameter() {parameterName=GrassColor Tint, defaultValue=<rgba 1 1 1 0>}.RGB, B=@n55.Default-Landscape-Color)
   Specular=@n55.Spec-CameraDistance
   Roughness=@n55.Spec-CameraDistance
   Normal=@n55.Spec-CameraDistance
  ) {attributeSetTypes [
    #1
    #2
    #3
    #4
   ]}
  n57 = Transform(None=NamedRerouteUsage())
  n58 = Multiply(A=NamedRerouteUsage(), B=Saturate(None=ComponentMask(None=@n57) {b=true}))
  n59 = Min(A=@n55.Mobile-Height, B=@n55.Mobile-Height)
  n60 = ScalarParameter() {defaultValue=1, group=Grass, parameterName=Grass Up Vector Contrast}
  NamedRerouteDeclaration(None=BlendMaterialAttributes(
    A=@n34
    B=@n56
    Alpha=MaterialFunctionCall(Transition Phase=@n58, Height Texture=@n59, Contrast=@n60) {materialFunction=&HeightLerp}.Alpha
   )) {name=Terrain Grass, nodeColor=<rgba 0.0547 1 0 1>}
  n61 = TextureSample(Tex=@n12)
  NamedRerouteDeclaration(None=StaticSwitchParameter(True=@n13.Normal Z Texture, False=@n61.RGB) {group=Rock Textures, parameterName="Base Texture WS Z "}) {name=Nmap, nodeColor=<rgba 0.3828 1 0 1>}
  NamedRerouteDeclaration(None=@n61.RGB) {name=HLODNormal, nodeColor=<rgba 0 0.5781 1 1>}
  n62 = GetMaterialAttributes(None=MaterialFunctionCall(In=MaterialFunctionCall!(In=@n35) {materialFunction=/Game/Environments/Apollo/Rocks/Natural_Rocks/Materials/MF_WSNormals}.Result) {materialFunction=&MF_DetailTexture}.Result) {attributeGetTypes [#4]}
  SetMaterialAttributes!(MaterialAttributes=@n62.MaterialAttributes, Normal=Transform(None=@n62.Normal)) {attributeSetTypes [#4]}
 ]
}
