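# Material function that supplies time-of-day color and opacity for distant fog silhouettes.
# Four vector parameters for morning, day, evening and night feed a Color_Time_of_Day call, which produces the SilhouetteColor output from their RGB only.
# A second Color_Time_of_Day call produces SilhouetteAlpha, using one shared scalar for morning, day and evening and a separate, higher-default scalar for night.
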MaterialFunction /Figment_S06/Effects/Fort_Effects/Materials/Smoke/MF_Distant_Fog_Silhouette_Colors
graph {nodes [
  FunctionOutput(None=MaterialFunctionCall(
    Color_Morning=VectorParameter() {parameterName=SilhouetteMorn, defaultValue=<rgba 0.1 0.0665 0.0966 0.7>}.RGB
    Color_Day=VectorParameter() {parameterName=SilhouetteDay, defaultValue=<rgba 0.061 0.1447 0.2 0.7>}.RGB
    Color_Evening=VectorParameter() {parameterName=SilhouetteEve, defaultValue=<rgba 0.0176 0.0303 0.05 0.7>}.RGB
    Color_Night=VectorParameter() {parameterName=SilhouetteNight, defaultValue=<rgba 0.0693 0.0608 0.1 0.7>}.RGB
   ) {materialFunction=&Color_Time_of_Day}.Result) {outputName=SilhouetteColor}
  n1 = ScalarParameter() {defaultValue=0.7, parameterName=SilhouetteAlpha}
  FunctionOutput(None=MaterialFunctionCall(Color_Morning=@n1, Color_Day=@n1, Color_Evening=@n1, Color_Night=ScalarParameter() {defaultValue=0.9, parameterName=SilhouetteAlphaNight}) {materialFunction=&Color_Time_of_Day}.Result) {outputName=SilhouetteAlpha, sortPriority=1}
 ]}
