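# Default Lit material for distant rock that layers snow over cave-wall rock and fades into water color by height.
# Snow masks combine up-facing normals and red vertex color, with diffuse alpha also breaking up the color mask.
# A world-Z mask flattens normals and blends to time-of-day water color below, while fog silhouette colors tint the rock.
# Feature-level switches reduce ES2/ES3.1 to flat normals, silhouette color and constant values.
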Material /Figment_S06/Athena/Environments/Materials/M_Distant_Rock
graph {
 properties {pixelDepthOffsetMode=PDOM_Legacy, usedShadingModels=Default Lit}
 nodes [
  n1 = Constant3Vector() {constant=<rgba 0 0 1 0>}
  n2 = TextureCoordinate() {uTiling=3, vTiling=3}
  n3 = TextureSample(UVs=TextureCoordinate() {uTiling=6, vTiling=6}) {samplerType=SAMPLERTYPE_Normal, texture=/Game/Packages/DS_Fortnite_Terrain_NoLOD/Textures/Cloud_Normal4}
  n4 = VectorParameter() {parameterName=NormalMult, defaultValue=<rgba 0.2 0.2 1 1>}
  n5 = ScalarParameter() {defaultValue=1.5, parameterName=SnowOpacity}
  n6 = Constant3Vector() {constant=<rgba 0 0 1 0>}
  n7 = DotProduct(A=VertexNormalWS(), B=@n6)
  n8 = ScalarParameter() {defaultValue=3, parameterName=SnowMaskPower}
  n9 = VertexColor()
  n10 = Add(A=Multiply(A=@n5, B=Power(Base=@n7, Exp=@n8)), B=@n9.R)
  n11 = Add(A=ComponentMask(None=WorldPosition().XYZ) {b=true}, B=ScalarParameter() {defaultValue=5000, parameterName=WaterHeightOffset})
  n12 = ScalarParameter() {defaultValue=14000, parameterName=WaterHeightBlend}
  n13 = Clamp(None=Power(Base=Divide(A=@n11, B=@n12), Exp=ScalarParameter() {defaultValue=1.5, parameterName=WaterBlendPower}))
  n14 = LinearInterpolate(
   A=@n1
   B=LinearInterpolate(
    A=TextureSample(UVs=@n2) {samplerType=SAMPLERTYPE_Normal, texture=/Game/Packages/DS_Fortnite_Terrain_NoLOD/Textures/T_Cave_Exterior_Wall_N}.RGB
    B=Multiply(A=@n3.RGB, B=@n4.RGB)
    Alpha=Clamp(None=@n10)
   )
   Alpha=@n13
  )
  FeatureLevelSwitch(Default=@n14, ES2=@n1, ES3_1=@n1, SM4_REMOVED=@n14)
  n15 = TextureSample(UVs=@n2) {texture=/Game/Packages/DS_Fortnite_Terrain_NoLOD/Textures/T_Cave_Exterior_Wall_D}
  n16 = OneMinus(None=@n15.A)
  n17 = PixelNormalWS()
  n18 = Multiply(A=Power(Base=DotProduct(A=@n17, B=@n6), Exp=@n8), B=@n5)
  n19 = Clamp(None=Add(A=Multiply(A=@n16) {constB=0.5}, B=Add(A=@n18, B=@n9.R)))
  n20 = Add(A=@n19, B=LinearInterpolate(Alpha=@n16) {constA=0.2, constB=0.6})
  n21 = Constant() {r=1}
  FeatureLevelSwitch(Default=@n20, ES2=@n21, ES3_1=@n21, SM4_REMOVED=@n20)
  n22 = VectorParameter() {parameterName=Morning, defaultValue=<rgba 0.1235 0.1422 0.1927 1>}
  n23 = VectorParameter() {parameterName=Day, defaultValue=<rgba 0.0384 0.1056 0.1615 1>}
  n24 = VectorParameter() {parameterName=Evening, defaultValue=<rgba 0.0648 0.1002 0.131 1>}
  n25 = VectorParameter() {parameterName=Night, defaultValue=<rgba 0.0304 0.0816 0.125 1>}
  n26 = Multiply(A=@n15.RGB, B=LinearInterpolate(Alpha=VertexColor().R) {constA=1, constB=1.5})
  n27 = VectorParameter() {parameterName=RockColor, defaultValue=<rgba 0.2592 0.6809 1 0>}
  n28 = Add(A=TextureSample(UVs=TextureCoordinate() {uTiling=16, vTiling=16}) {texture=/Game/Packages/DS_Fortnight/WhiteRockCustom_D}.RGB) {constB=0.33}
  n29 = VectorParameter() {parameterName=SnowColor, defaultValue=<rgba 1.2094 1.3244 1.5 0>}
  n30 = MaterialFunctionCall() {materialFunction=/Figment_S06/Effects/Fort_Effects/Materials/Smoke/MF_Distant_Fog_Silhouette_Colors}
  n31 = LinearInterpolate(
   A=LinearInterpolate(A=Multiply(A=@n26, B=@n27.RGB), B=Multiply(A=@n28, B=@n29.RGB), Alpha=@n19)
   B=@n30.SilhouetteColor
   Alpha=@n30.SilhouetteColor
  ) {constAlpha=0.7}
  LinearInterpolate(
   A=Multiply(
    A=MaterialFunctionCall(Color_Morning=@n22.RGB, Color_Day=@n23.RGB, Color_Evening=@n24.RGB, Color_Night=@n25.RGB) {materialFunction=&Color_Time_of_Day}.Result
    B=VectorParameter() {parameterName=WaterColor, defaultValue=<rgba 0.1758 0.4661 0.75 1>}.RGB
   )
   B=FeatureLevelSwitch(Default=@n31, ES2=@n30.SilhouetteColor, ES3_1=@n30.SilhouetteColor, SM4_REMOVED=@n31)
   Alpha=@n13
  )
 ]
}
