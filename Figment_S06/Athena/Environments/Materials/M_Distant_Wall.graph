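# M_Distant_Wall is a Default Lit material for distant wall scenery, with procedural snow and fog silhouette tinting.
# One branch scales a light color by vertex red and a time-of-day collection component, then adds the silhouette color from a distant fog function.
# The other blends the wall texture toward snow color using pivot-relative height and upward-facing normals, then toward that silhouette color.
# A local SilhouetteColor parameter is left unconnected.
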Material /Figment_S06/Athena/Environments/Materials/M_Distant_Wall
graph {
 properties {pixelDepthOffsetMode=PDOM_Legacy, previewMesh=/Figment_S06/Athena/Environments/Meshes/SM_Fort_Low, usedShadingModels=Default Lit}
 nodes [
  n1 = CollectionParameter() {
   collection=&FortniteMaterialParameters
   parameterName=FortniteActiveTimeOfDay
  }
  n2 = MaterialFunctionCall(Float4=@n1) {materialFunction=&BreakOutFloat4Components}
  n3 = VertexColor()
  n4 = MaterialFunctionCall() {materialFunction=/Figment_S06/Effects/Fort_Effects/Materials/Smoke/MF_Distant_Fog_Silhouette_Colors}
  Add(
   A=Multiply(A=VectorParameter() {parameterName=LightColor, defaultValue=<rgba 16 4.5388 0 0>}.RGB, B=Multiply(A=@n2.A, B=@n3.R))
   B=@n4.SilhouetteColor
  )
  n5 = Subtract(A=ObjectPositionWS(), B=WorldPosition().XYZ)
  n6 = Divide(
   A=Add(A=ComponentMask(None=@n5) {b=true}, B=ScalarParameter() {defaultValue=400, parameterName=SnowOffset}) {constB=300}
   B=ScalarParameter() {defaultValue=800, parameterName=SnowFade}
  ) {constB=400}
  n7 = DotProduct(A=VertexNormalWS(), B=Constant3Vector() {constant=<rgba 0 0 1 0>})
  n8 = Multiply(A=Power(Base=Add(A=@n6, B=@n7)), B=ScalarParameter() {defaultValue=1.5, parameterName=SnowOpacity})
  LinearInterpolate(
   A=LinearInterpolate(
    A=TextureSample() {texture=/Figment_S06/Athena/Environments/Textures/T_Distant_Wall}.RGB
    B=VectorParameter() {parameterName=SnowColor, defaultValue=<rgba 0.8062 0.8829 1 0>}.RGB
    Alpha=Clamp(None=@n8)
   )
   B=@n4.SilhouetteColor
   Alpha=@n4.SilhouetteColor
  ) {constAlpha=0.7}
  VectorParameter() {parameterName=SilhouetteColor, defaultValue=<rgba 0.0305 0.0723 0.1 0.7>}
 ]
}
